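# Sound cue for the button click of the C4 remote detonator.
# A mixer layers two waves at once: the remote button press at full level and the C4 trigger sound mixed much quieter.
# It plays as weapon foley with weapon-reload attenuation, reduced overall volume and a concurrency override that caps simultaneous instances.
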SoundCue /Game/Athena/Sounds/Items/C4/C4_Detonate_Button_Click_Cue
root=Mixer_0
nodes [
 {id=Mixer_0, children [WavePlayer_1,WavePlayer_0], properties {inputVolume [0.1,1]}}
 {id=WavePlayer_0, properties {soundWaveAssetPtr=&/C4_Detonate_Remote_Button_Press_01}}
 {id=WavePlayer_1, properties {soundWaveAssetPtr=&/C4_Trigger_01}}
]
properties {
 attenuationSettings=&Fort_Weapon_Reload_Att
 bOverrideConcurrency=true
 maxDistance=2250
 soundClassObject=&Fort_WeaponFoley
 volumeMultiplier=0.501
 concurrencyOverrides {maxCount=3}
}
