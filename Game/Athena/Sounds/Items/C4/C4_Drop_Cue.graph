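# Sound cue for the C4 item's drop sound in Athena.
# A single wave player feeds the C4_Drop_01 wave straight to the output at reduced volume, with no randomization or modulation.
# It uses the shared weapon-drop attenuation and the gadget items sound class.
# A concurrency override blocks new instances once the limit is reached rather than cutting off playing ones.
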SoundCue /Game/Athena/Sounds/Items/C4/C4_Drop_Cue
root=WavePlayer_1
nodes [{id=WavePlayer_1, properties {soundWaveAssetPtr=&/C4_Drop_01}}]
properties {
 attenuationSettings=&Fort_Weapon_Drop_Att
 bOverrideConcurrency=true
 maxDistance=3000
 soundClassObject=&Fort_Gadgets_Items
 volumeMultiplier=0.5
 concurrencyOverrides {resolutionRule=PreventNew}
}
