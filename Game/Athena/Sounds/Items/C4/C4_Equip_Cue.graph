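# Sound cue for equipping the C4 item in Athena.
# A single wave player feeds the root directly, playing one equip sample with no randomization and a cue-level pitch and volume adjustment.
# It uses the weapon foley sound class, borrows the weapon reload attenuation, and overrides concurrency to limit simultaneous instances.
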SoundCue /Game/Athena/Sounds/Items/C4/C4_Equip_Cue
root=WavePlayer_1
nodes [{id=WavePlayer_1, properties {soundWaveAssetPtr=&/C4_Equip_01}}]
properties {
 attenuationSettings=&Fort_Weapon_Reload_Att
 bOverrideConcurrency=true
 maxDistance=2250
 pitchMultiplier=2
 soundClassObject=&Fort_WeaponFoley
 volumeMultiplier=0.652
 concurrencyOverrides {maxCount=3}
}
