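# Sound cue for the C4 explosion in Athena, layering close, medium and distant variants by range.
# Each layer randomly picks one of four waves, then passes through a modulator, its own attenuation and a sound class override into a mixer that weights close loudest.
# The distant layer adds a short random delay, and the close random node is excluded from branch culling.
# A concurrency override allows one instance per owner.
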SoundCue /Game/Athena/Sounds/Items/C4/C4_Explosion_Cue
root=Mixer_1
nodes [
 {id=Attenuation_10, children [Modulator_13], properties {attenuationSettings=/Game/Sounds/Attenuation/Fort_Explosion_Medium_Athena_Att}}
 {
  id=Attenuation_11
  children [Delay_0]
  properties {
   attenuationSettings=/Game/Sounds/Attenuation/Fort_Explosion_Distant_Athena_Att
   attenuationOverrides {falloffDistance=3000, attenuationShapeExtents {x=1000}}
  }
 }
 {
  id=Attenuation_9
  children [Modulator_2]
  properties {
   attenuationSettings=/Game/Sounds/Attenuation/Fort_Explosion_Close_Athena_Att
   attenuationOverrides {falloffDistance=900, attenuationShapeExtents {x=100}}
  }
 }
 {id=Delay_0, children [Modulator_1], properties {delayMax=0.1}}
 {id=Mixer_1, children [SoundClass_6,SoundClass_2,SoundClass_0], properties {inputVolume [2,1.25,1.25]}}
 {id=Modulator_1, children [Random_6], properties {pitchMax=1.2, pitchMin=0.85, volumeMax=1, volumeMin=1}}
 {id=Modulator_13, children [Random_5], properties {pitchMax=1.1, pitchMin=0.9, volumeMax=1, volumeMin=1}}
 {id=Modulator_2, children [Random_1], properties {volumeMax=1, volumeMin=1}}
 {
  id=Random_1
  children [WavePlayer_21, WavePlayer_22, WavePlayer_23, WavePlayer_24]
  properties {bShouldExcludeFromBranchCulling=true, weights [1,1,1,1]}
 }
 {id=Random_5, children [WavePlayer_25, WavePlayer_26, WavePlayer_27, WavePlayer_28], properties {weights [1,1,1,1]}}
 {id=Random_6, children [WavePlayer_29, WavePlayer_30, WavePlayer_31, WavePlayer_32], properties {weights [1,1,1,1]}}
 {
  id=SoundClass_0
  children [Attenuation_11]
  properties {soundClassOverride=&Fort_WeaponRanged_Fire_Distant}
 }
 {
  id=SoundClass_2
  children [Attenuation_10]
  properties {soundClassOverride=&Fort_WeaponRanged_Fire_Distant}
 }
 {
  id=SoundClass_6
  children [Attenuation_9]
  properties {soundClassOverride=&Fort_WeaponRanged_Fire_3P_PSM}
 }
 {id=WavePlayer_21, properties {soundWaveAssetPtr=&/C4_Explode_Close_01}}
 {id=WavePlayer_22, properties {soundWaveAssetPtr=&/C4_Explode_Close_02}}
 {id=WavePlayer_23, properties {soundWaveAssetPtr=&/C4_Explode_Close_03}}
 {id=WavePlayer_24, properties {soundWaveAssetPtr=&/C4_Explode_Close_04}}
 {id=WavePlayer_25, properties {soundWaveAssetPtr=&/C4_Explode_Medium_01}}
 {id=WavePlayer_26, properties {soundWaveAssetPtr=&/C4_Explode_Medium_02}}
 {id=WavePlayer_27, properties {soundWaveAssetPtr=&/C4_Explode_Medium_03}}
 {id=WavePlayer_28, properties {soundWaveAssetPtr=&/C4_Explode_Medium_04}}
 {id=WavePlayer_29, properties {soundWaveAssetPtr=&/C4_Explode_Distant_01}}
 {id=WavePlayer_30, properties {soundWaveAssetPtr=&/C4_Explode_Distant_02}}
 {id=WavePlayer_31, properties {soundWaveAssetPtr=&/C4_Explode_Distant_03}}
 {id=WavePlayer_32, properties {soundWaveAssetPtr=&/C4_Explode_Distant_04}}
]
properties {bOverrideConcurrency=true, maxDistance=27500, volumeMultiplier=1, concurrencyOverrides {bLimitToOwner=true, maxCount=1}}
