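# Sound cue for the C4 item's landing sound in Athena.
# A single wave player feeds the C4_Land_01 wave straight to the output, pitched up and lowered in volume.
# It uses the shared weapon pickup attenuation and the gadgets items sound class.
# A concurrency override prevents new instances from starting once the limit is reached, rather than stopping playing ones.
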SoundCue /Game/Athena/Sounds/Items/C4/C4_Land_Cue
root=WavePlayer_1
nodes [{id=WavePlayer_1, properties {soundWaveAssetPtr=&/C4_Land_01}}]
properties {
 attenuationSettings=&Fort_Weapon_Pickup_Att
 bOverrideConcurrency=true
 maxDistance=3500
 pitchMultiplier=1.3
 soundClassObject=&Fort_Gadgets_Items
 volumeMultiplier=0.5
 concurrencyOverrides {resolutionRule=PreventNew}
}
