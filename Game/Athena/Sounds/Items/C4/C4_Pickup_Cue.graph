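# Sound cue for picking up a C4 item in Athena, playing a single C4 pickup wave through one wave player node.
# It uses the weapon pickup attenuation and the gadget items sound class, at a reduced volume.
# Its own concurrency override blocks new instances at the limit rather than stopping sounds already playing.
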SoundCue /Game/Athena/Sounds/Items/C4/C4_Pickup_Cue
root=WavePlayer_1
nodes [{id=WavePlayer_1, properties {soundWaveAssetPtr=/Game/Sounds/Fort_Gadgets/Unassigned/C4_Pickup_01}}]
properties {
 attenuationSettings=&Fort_Weapon_Pickup_Att
 bOverrideConcurrency=true
 maxDistance=3500
 soundClassObject=&Fort_Gadgets_Items
 volumeMultiplier=0.3
 concurrencyOverrides {resolutionRule=PreventNew}
}
