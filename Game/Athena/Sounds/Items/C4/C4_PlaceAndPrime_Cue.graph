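# C4_PlaceAndPrime_Cue is a Sound Cue, named for placing and priming the C4 item.
# A Random node picks between two C4 placement waves with equal weight.
# The cue routes to the gadget items sound class with trap attenuation and reduced volume, and its concurrency override prevents new instances once the limit is reached.
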SoundCue /Game/Athena/Sounds/Items/C4/C4_PlaceAndPrime_Cue
root=Random_0
nodes [
 {id=Random_0, children [WavePlayer_0,WavePlayer_2], properties {weights [1,1]}}
 {id=WavePlayer_0, properties {soundWaveAssetPtr=&/C4_Place_01}}
 {id=WavePlayer_2, properties {soundWaveAssetPtr=&/C4_Place_02}}
]
properties {
 attenuationSettings=&Fort_Traps_Att
 bOverrideConcurrency=true
 maxDistance=6000
 soundClassObject=&Fort_Gadgets_Items
 volumeMultiplier=0.6
 concurrencyOverrides {resolutionRule=PreventNew}
}
