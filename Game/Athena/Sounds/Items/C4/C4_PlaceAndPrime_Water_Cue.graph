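# Sound cue for placing and priming C4 in water, classed as a gadget item sound.
# A random node picks one of three equally weighted wave variants and feeds a modulator that holds pitch fixed.
# Playback runs at reduced volume with shared trap attenuation settings.
# A concurrency override blocks new instances once the limit is reached rather than stopping ones already playing.
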SoundCue /Game/Athena/Sounds/Items/C4/C4_PlaceAndPrime_Water_Cue
root=Modulator_1
nodes [
 {id=Modulator_1, children [Random_1], properties {pitchMax=1, pitchMin=1}}
 {id=Random_1, children [WavePlayer_3,WavePlayer_4,WavePlayer_5], properties {weights [1,1,1]}}
 {id=WavePlayer_3, properties {soundWaveAssetPtr=&/C4_PlaceAndPrime_Water_01}}
 {id=WavePlayer_4, properties {soundWaveAssetPtr=&/C4_PlaceAndPrime_Water_02}}
 {id=WavePlayer_5, properties {soundWaveAssetPtr=&/C4_PlaceAndPrime_Water_03}}
]
properties {
 attenuationSettings=&Fort_Traps_Att
 bOverrideConcurrency=true
 maxDistance=6000
 soundClassObject=&Fort_Gadgets_Items
 volumeMultiplier=0.6
 concurrencyOverrides {resolutionRule=PreventNew}
}
