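# C4 pre-explosion sound cue for Athena, the warning sound played before a C4 charge detonates.
# A single wave player feeds the C4_PreExplosion wave straight to the output, with reduced volume.
# It uses close-range weapon projectile attenuation and the gadget items sound class.
# A concurrency override limits it to one instance at a time.
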SoundCue /Game/Athena/Sounds/Items/C4/C4_PreExplode_Cue
root=WavePlayer_1
nodes [{id=WavePlayer_1, properties {soundWaveAssetPtr=&/C4_PreExplosion}}]
properties {
 attenuationSettings=&Fort_Weapon_Projectile_Close_Athena_Att
 bOverrideConcurrency=true
 maxDistance=4950
 soundClassObject=&Fort_Gadgets_Items
 volumeMultiplier=0.4
 concurrencyOverrides {maxCount=1}
}
