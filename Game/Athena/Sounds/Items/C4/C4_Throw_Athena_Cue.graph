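# Sound cue for the Athena C4 throw, reusing player run foley as the throw sound.
# A random node picks one of three equally weighted foley waves and passes it through a modulator at the root.
# The cue plays at reduced volume in the gadgets and items sound class, with gadget-placement attenuation.
# A concurrency override prevents new instances from starting once the limit is reached.
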SoundCue /Game/Athena/Sounds/Items/C4/C4_Throw_Athena_Cue
root=Modulator_0
nodes [
 {id=Modulator_0, children [Random_0]}
 {id=Random_0, children [WavePlayer_1,WavePlayer_2,WavePlayer_3], properties {weights [1,1,1]}}
 {id=WavePlayer_1, properties {soundWaveAssetPtr=/Game/Sounds/Fort_Foley_Sounds/Moves/player_run_foley_11}}
 {id=WavePlayer_2, properties {soundWaveAssetPtr=/Game/Sounds/Fort_Foley_Sounds/Moves/player_run_foley_12}}
 {id=WavePlayer_3, properties {soundWaveAssetPtr=/Game/Sounds/Fort_Foley_Sounds/Moves/player_run_foley_02}}
]
properties {
 attenuationSettings=/Game/Sounds/Attenuation/Fort_GadgetPlace_Att
 bOverrideConcurrency=true
 maxDistance=2000
 soundClassObject=&Fort_Gadgets_Items
 volumeMultiplier=0.3
 concurrencyOverrides {resolutionRule=PreventNew}
}
